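# Material function for the Grilled Cheese backpack that draws a glowing vortex sprite effect, output as Stage 2, 3 and 4 colors plus two flame noises.
# Screen-aligned UVs center on the socket position, offset by a warp texture per stage switches and scaled by FOV.
# These UVs drive a pulsing sphere-mask glow, twisted warped cloud layers and noise-distorted rings.
# Later stages combine more layers at greater strength.
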MaterialFunction /Game/Accessories/FORT_Backpacks/Backpack_M_MED_Grilled_Cheese/Materials/MF_Backpack_Sprite
graph {nodes [
  n1 = FunctionInput(Preview=VectorParameter() {group=GrilledCheese, parameterName=SocketLocation, sortPriority=601, defaultValue=<rgba 0 0 0 1>}.RGB) {inputName=In Socket UVs}
  n2 = TextureObjectParameter() {
   group=Grilled_Cheese_Textures
   parameterName=GC_OcclusionTexture
   sortPriority=300
   texture=/Game/Accessories/FORT_Backpacks/Backpack_M_MED_Grilled_Cheese/Textures/T_M_MED_GrilledCheese_Backpack_M2
  }
  n3 = TextureSample(Tex=@n2) {texture=&T_BlacksRGB}
  n4 = Add(
   A=@n1
   B=Multiply(
    A=FunctionInput(Preview=@n3.RGB) {inputName=Warp Texture}
    B=ScalarParameter() {defaultValue=0.15, group=GrilledCheese_FX, parameterName=GC_OffsetAmount, sortPriority=-1}
   )
  )
  n5 = LinearInterpolate(
   A=LinearInterpolate(A=@n1, B=@n4, Alpha=ScalarParameter() {group=GrilledCheese, parameterName=GC_UseStage2, sortPriority=-1})
   B=@n4
   Alpha=ScalarParameter() {group=GrilledCheese, parameterName=GC_UseStage3, sortPriority=-1}
  )
  n6 = ScalarParameter() {group=GrilledCheese, parameterName=GC_UseStage4, sortPriority=-1}
  n7 = ScalarParameter() {defaultValue=12, group=GrilledCheese_FX, parameterName=GC_UV_FXScale}
  n8 = MaterialFunctionCall!(
   Input=MaterialFunctionCall() {materialFunction=/Engine/Functions/Engine_MaterialFunctions02/Utility/FOV}."FOV "
   Input Low=Constant() {r=0.414}
   Input High=Constant() {r=1.091}
   Target Low=Constant()
   Target Height=Constant() {r=1}
  ) {materialFunction=&RemapValueRange}
  n9 = MaterialFunctionCall(
   UVs=MaterialFunctionCall!(Center Point World Position=LinearInterpolate(A=@n5, B=@n4, Alpha=@n6)) {materialFunction=/Engine/Functions/Engine_MaterialFunctions02/UVs/WorldSpaceAlignedScreenCoordinates}.Offset UVs
   Texture Scale=LinearInterpolate(A=Multiply(A=@n7) {constB=2}, B=@n7, Alpha=Saturate(None=@n8.Result))
  ) {materialFunction=&ScaleUVsByCenter}
  n10 = Subtract(A=Multiply(A=@n9.UVs) {constB=2})
  n11 = Constant2Vector()
  n12 = Time()
  n13 = Add(A=Sine(None=Multiply(A=@n12) {constB=0.2}) {period=0.5})
  n14 = LinearInterpolate(
   A=ScalarParameter() {defaultValue=1.2, group=GrilledCheese_FX, parameterName=GC_GlowMinRadius}
   B=ScalarParameter() {defaultValue=2, group=GrilledCheese_FX, parameterName=GC_GlowMaxRadius}
   Alpha=Clamp(None=Divide(A=@n13)) {minDefault=0.01}
  )
  n15 = Constant()
  n16 = Clamp(None=Power(Base=SphereMask(A=@n10, B=@n11, Radius=@n14, Hardness=@n15)))
  n17 = ScalarParameter() {defaultValue=1, group=GrilledCheese_FX, parameterName=GC_OuterGlowStrength}
  n18 = Multiply(
   A=Multiply(A=VectorParameter() {group=GrilledCheese_FX, parameterName=GC_Stage3_CoreGlow_Color, defaultValue=<rgba 1 0.0047 0 1>}.RGB, B=@n16)
   B=@n17
  )
  n19 = Constant3Vector()
  n20 = VectorParameter() {group=GrilledCheese_FX_Vortex, parameterName=GC_VortexD_Color_Dark, defaultValue=<rgba 0.3452 0.1199 0.0586 1>}
  n21 = VectorParameter() {group=GrilledCheese_FX_Vortex, parameterName=GC_VortexD_Color_Light, defaultValue=<rgba 0.6637 0.526 0.16 1>}
  n22 = ScalarParameter() {defaultValue=0.5, group=GrilledCheese_FX_Vortex, parameterName=GC_Master_tile_Size}
  n23 = Subtract(A=Multiply(A=@n9.UVs) {constB=2})
  n24 = Constant2Vector()
  n25 = ScalarParameter() {defaultValue=3, group=GrilledCheese_FX_Vortex, parameterName=GC_SwirlRadius}
  n26 = SphereMask(A=@n23, B=@n24, Radius=@n25, Hardness=Constant())
  n27 = Multiply(A=@n26, B=ScalarParameter() {defaultValue=15, group=GrilledCheese_FX_Vortex, parameterName=GC_SwirlCoreStrength})
  n28 = Multiply(A=@n22, B=Rotator(Coordinate=@n23, Time=@n27))
  n29 = Multiply(A=@n28, B=Constant() {r=0.75}) {constB=0.75}
  n30 = VectorParameter() {group=GrilledCheese_FX_Vortex, parameterName=GC_Clouds_B_Speed, defaultValue=<rgba -0.08 0.02 0 1>}
  n31 = Multiply(
   A=Time()
   B=ComponentMask(None=VectorParameter() {group=GrilledCheese_FX_Vortex, parameterName=GC_Clouds_WarpB_Speed, defaultValue=<rgba -0.1 -0.03 0 1>}.RGB) {g=true, r=true}
  )
  n32 = TextureObject() {texture=/Game/Effects/Fort_Effects/Textures/Energy/T_OutpostShockwaveStreaks_02}
  n33 = Subtract(A=TextureSample(UVs=Add(A=@n28, B=@n31), Tex=@n32) {texture=&T_BlacksRGB}.R) {constB=0.2}
  n34 = ScalarParameter() {defaultValue=0.25, group=GrilledCheese_FX_Vortex, parameterName=GC_Clouds_WarpB_Amount}
  n35 = TextureSample(UVs=Add(A=Panner(Coordinate=@n29, Speed=@n30.RGB), B=Multiply(A=@n33, B=@n34) {constB=0.1}), Tex=@n32) {texture=&T_BlacksRGB}
  n36 = ScalarParameter() {defaultValue=1, group=GrilledCheese_FX_Vortex, parameterName=GC_CloudHeightPowerForDiffuse}
  n37 = VectorParameter() {group=GrilledCheese_FX_Vortex, parameterName=GC_Clouds_A_Speed, defaultValue=<rgba 0.1 -0.03 0 1>}
  n38 = Multiply(A=@n22, B=Rotator(Coordinate=@n23, Time=@n27))
  n39 = Multiply(
   A=Time()
   B=ComponentMask(None=VectorParameter() {group=GrilledCheese_FX_Vortex, parameterName=GC_Clouds_WarpA_Speed, defaultValue=<rgba -0.1 -0.03 0 1>}.RGB) {g=true, r=true}
  )
  n40 = Subtract(A=TextureSample(UVs=Add(A=@n38, B=@n39), Tex=@n32) {texture=&T_BlacksRGB}.R) {constB=0.2}
  n41 = ScalarParameter() {defaultValue=0.25, group=GrilledCheese_FX_Vortex, parameterName=GC_Clouds_WarpA_Amount}
  n42 = TextureSample(UVs=Add(A=Panner(Coordinate=@n38, Speed=@n37.RGB), B=Multiply(A=@n40, B=@n41) {constB=0.1}), Tex=@n32) {texture=&T_BlacksRGB}
  n43 = ScalarParameter() {defaultValue=1, group=GrilledCheese_FX_Vortex, parameterName=GC_Clouds_Masking_Multipler}
  n44 = Saturate(None=Add(A=Multiply(A=@n42.R, B=@n43) {constB=1.5}, B=ScalarParameter() {group=GrilledCheese_FX_Vortex, parameterName=GC_Clouds_Masking_contrast}) {constB=-0.5})
  n45 = LinearInterpolate(
   A=LinearInterpolate(A=@n20.RGB, B=@n21.RGB, Alpha=Power(Base=@n35.R, Exp=@n36))
   B=LinearInterpolate(
    A=VectorParameter() {group=GrilledCheese_FX_Vortex, parameterName=GC_VortexC_Color_Dark, defaultValue=<rgba 1 0.214 0 1>}.RGB
    B=VectorParameter() {group=GrilledCheese_FX_Vortex, parameterName=GC_VortexC_Color_Light, defaultValue=<rgba 1 0.7267 0 1>}.RGB
    Alpha=Power(Base=@n42.R, Exp=@n36)
   )
   Alpha=@n44
  )
  n46 = VectorParameter() {group=GrilledCheese_FX_Vortex, parameterName=GC_VortexB_Color_Dark, defaultValue=<rgba 0.3452 0.1199 0.0586 1>}
  n47 = VectorParameter() {group=GrilledCheese_FX_Vortex, parameterName=GC_VortexB_Color_Light, defaultValue=<rgba 0.6637 0.526 0.16 1>}
  n48 = Time()
  n49 = Multiply(A=@n48, B=ScalarParameter() {defaultValue=3, group=GrilledCheese_FX_Vortex, parameterName=GC_SwirlCoreB_Speed})
  n50 = Multiply(A=@n26, B=ScalarParameter() {defaultValue=15, group=GrilledCheese_FX_Vortex, parameterName=GC_SwirlCoreInnerStrength})
  n51 = Multiply(A=@n22, B=Rotator(Coordinate=@n23, Time=Add(A=@n49, B=@n50)) {speed=-0.25})
  n52 = Multiply(A=@n51, B=Constant() {r=0.75}) {constB=0.75}
  n53 = VectorParameter() {group=GrilledCheese_FX_Vortex, parameterName=GC_Clouds_B_Speed, defaultValue=<rgba -0.08 0.02 0 1>}
  n54 = Multiply(
   A=Time()
   B=ComponentMask(None=VectorParameter() {group=GrilledCheese_FX_Vortex, parameterName=GC_Clouds_WarpB_Speed, defaultValue=<rgba -0.1 -0.03 0 1>}.RGB) {g=true, r=true}
  )
  n55 = Subtract(A=TextureSample(UVs=Add(A=@n51, B=@n54), Tex=@n32) {texture=&T_BlacksRGB}.R) {constB=0.2}
  n56 = ScalarParameter() {defaultValue=0.25, group=GrilledCheese_FX_Vortex, parameterName=GC_Clouds_WarpB_Amount}
  n57 = TextureSample(UVs=Add(A=Panner(Coordinate=@n52, Speed=@n53.RGB), B=Multiply(A=@n55, B=@n56) {constB=0.1}), Tex=@n32) {texture=&T_BlacksRGB}
  n58 = ScalarParameter() {defaultValue=1, group=GrilledCheese_FX_Vortex, parameterName=GC_CloudHeightPowerForDiffuse}
  n59 = VectorParameter() {group=GrilledCheese_FX_Vortex, parameterName=GC_Clouds_A_Speed, defaultValue=<rgba 0.1 -0.03 0 1>}
  n60 = Multiply(A=@n48, B=ScalarParameter() {defaultValue=5, group=GrilledCheese_FX_Vortex, parameterName=GC_SwirlCoreA_Speed})
  n61 = Multiply(A=@n22, B=Rotator(Coordinate=@n23, Time=Add(A=@n60, B=@n50)) {speed=-0.25})
  n62 = Multiply(
   A=Time()
   B=ComponentMask(None=VectorParameter() {group=GrilledCheese_FX_Vortex, parameterName=GC_Clouds_WarpA_Speed, defaultValue=<rgba -0.1 -0.03 0 1>}.RGB) {g=true, r=true}
  )
  n63 = Subtract(A=TextureSample(UVs=Add(A=@n61, B=@n62), Tex=@n32) {texture=&T_BlacksRGB}.R) {constB=0.2}
  n64 = ScalarParameter() {defaultValue=0.25, group=GrilledCheese_FX_Vortex, parameterName=GC_Clouds_WarpA_Amount}
  n65 = TextureSample(UVs=Add(A=Panner(Coordinate=@n61, Speed=@n59.RGB), B=Multiply(A=@n63, B=@n64) {constB=0.1}), Tex=@n32) {texture=&T_BlacksRGB}
  n66 = ScalarParameter() {defaultValue=1, group=GrilledCheese_FX_Vortex, parameterName=GC_Clouds_Masking_Multipler}
  n67 = Saturate(None=Add(A=Multiply(A=@n65.R, B=@n66) {constB=1.5}, B=ScalarParameter() {group=GrilledCheese_FX_Vortex, parameterName=GC_Clouds_Masking_contrast}) {constB=-0.5})
  n68 = LinearInterpolate(
   A=LinearInterpolate(A=@n46.RGB, B=@n47.RGB, Alpha=Power(Base=@n57.R, Exp=@n58))
   B=LinearInterpolate(
    A=VectorParameter() {group=GrilledCheese_FX_Vortex, parameterName=GC_VortexA_Color_Dark, defaultValue=<rgba 1 0.214 0 1>}.RGB
    B=VectorParameter() {group=GrilledCheese_FX_Vortex, parameterName=GC_VortexA_Color_Light, defaultValue=<rgba 1 0.7267 0 1>}.RGB
    Alpha=Power(Base=@n65.R, Exp=@n58)
   )
   Alpha=@n67
  )
  n69 = Saturate(None=Multiply(A=@n26, B=ScalarParameter() {defaultValue=1, group=GrilledCheese_FX_Vortex, parameterName=GC_SwirlCoreMask_Intensity}))
  n70 = Subtract(A=@n25, B=ScalarParameter() {defaultValue=1, group=GrilledCheese_FX_Vortex, parameterName=GC_SwirlRadius_Mask})
  n71 = Constant()
  n72 = LinearInterpolate(A=@n35.R, B=@n42.R, Alpha=@n44)
  n73 = Multiply(A=@n69, B=@n72)
  n74 = LinearInterpolate(A=@n19, B=LinearInterpolate(A=@n45, B=@n68, Alpha=@n69), Alpha=Add(A=SphereMask(A=@n23, B=@n24, Radius=@n70, Hardness=@n71), B=@n73))
  n75 = SphereMask(A=Subtract(A=Multiply(A=@n9.UVs) {constB=2}), B=Constant2Vector(), Radius=Constant() {r=1.75}, Hardness=Constant())
  n76 = VectorParameter() {group=GrilledCheese_FX_Waves, parameterName=GC_CoreDistortionNoise_Tiling, defaultValue=<rgba 0.5 0.5 0.5 1>}
  n77 = ComponentMask(None=@n76.RGB) {g=true, r=true}
  n78 = VectorParameter() {group=GrilledCheese_FX_Waves, parameterName=GC_CoreDistortionNoise_PanSpeedAndDirection, defaultValue=<rgba 0.15 0.15 1 1>}
  n79 = TextureSampleParameter2D(UVs=Panner(Coordinate=Multiply(A=@n9.UVs, B=@n77), Speed=ComponentMask(None=@n78.RGB) {g=true, r=true})) {
   group=GrilledCheese_FX_Waves
   parameterName=DistortTexture
   samplerType=SAMPLERTYPE_LinearColor
   sortPriority=-100
   texture=&T_NoisePack_LinearColor
  }
  n80 = Add(
   A=Time()
   B=Multiply(A=@n79.G, B=ScalarParameter() {defaultValue=0.25, group=GrilledCheese_FX_Waves, parameterName=GC_CoreWaveTimeDistortion})
  )
  n81 = ScalarParameter() {defaultValue=0.35, group=GrilledCheese_FX_Waves, parameterName=GC_CoreWaves_Speed}
  n82 = Multiply(
   A=Add(A=@n75, B=Multiply(A=@n80, B=@n81))
   B=ScalarParameter() {defaultValue=4, group=GrilledCheese_FX_Waves, parameterName=GC_CoreWave_Frequency}
  )
  n83 = Power(Base=Subtract(A=Frac(None=@n82)) {constB=0.5})
  n84 = Clamp(None=Multiply(A=Power(Base=@n75), B=Multiply(A=@n83) {constB=4}))
  n85 = SphereMask(A=Subtract(A=Multiply(A=@n9.UVs) {constB=2}), B=Constant2Vector(), Radius=Constant() {r=2}, Hardness=Constant())
  n86 = Multiply(
   A=Multiply(
    A=VectorParameter() {group=GrilledCheese_FX_Waves, parameterName=GC_CoreWaves_Color, defaultValue=<rgba 0.5 0.2804 0 1>}.RGB
    B=Multiply(A=@n84, B=@n85)
   )
   B=ScalarParameter() {defaultValue=1, group=GrilledCheese_FX_Waves, parameterName=GC_CoreWaveGlowStrength}
  ) {constB=0.5}
  n87 = VectorParameter() {group=GrilledCheese_FX_Waves, parameterName=GC_Stage4_CoreWaves_Color, defaultValue=<rgba 0.5 0.4221 0.1785 1>}
  n88 = SphereMask(A=Subtract(A=Multiply(A=@n9.UVs) {constB=2}), B=Constant2Vector(), Radius=Constant() {r=1.75}, Hardness=Constant())
  n89 = Add(A=Add(A=Time()) {constB=0.25}, B=@n79.G)
  n90 = ScalarParameter() {defaultValue=0.55, group=GrilledCheese_FX_Waves, parameterName=GC_Stage4_CoreRings_Speed}
  n91 = Multiply(
   A=Add(A=@n88, B=Multiply(A=@n89, B=@n90))
   B=ScalarParameter() {defaultValue=3, group=GrilledCheese_FX_Waves, parameterName=GC_Stage4_CoreRings_Frequency}
  )
  n92 = Power(Base=Subtract(A=Frac(None=@n91)) {constB=0.5})
  n93 = Clamp(None=Multiply(A=Power(Base=@n88), B=Multiply(A=@n92) {constB=10}))
  n94 = SphereMask(A=Subtract(A=Multiply(A=@n9.UVs) {constB=2}), B=Constant2Vector(), Radius=Constant() {r=2}, Hardness=Constant())
  n95 = Multiply(
   A=Multiply(A=@n87.RGB, B=Multiply(A=@n93, B=@n94))
   B=ScalarParameter() {defaultValue=1, group=GrilledCheese_FX_Waves, parameterName=GC_Stage4_RingsGlowStrength}
  ) {constB=0.5}
  FunctionOutput(None=Add(A=Add(A=@n18, B=@n74), B=Add(A=@n86, B=@n95))) {outputName=Stage 4, sortPriority=3}
  FunctionOutput(None=@n79.B) {outputName=Flame Noise 02}
  n96 = VectorParameter() {group=GrilledCheese_FX, parameterName=GC_Stage2_CoreGlow_Color, defaultValue=<rgba 0.3 0.0958 0.0073 0.5>}
  n97 = Multiply(A=Multiply(A=@n96.RGB, B=@n16), B=@n17)
  n98 = LinearInterpolate(A=@n65.R, B=@n57.R, Alpha=@n67)
  n99 = Multiply(A=@n69, B=@n98)
  n100 = ScalarParameter() {defaultValue=0.025, desc=uses Vortex A+B, group=GrilledCheese, parameterName=GC_Stage3_WhispsMultiplier, sortPriority=0}
  n101 = Add(A=Multiply(A=LinearInterpolate(A=@n19, B=@n68, Alpha=@n99)) {constB=0.7}, B=Multiply(A=@n73, B=@n100) {constB=0.025})
  n102 = ScalarParameter() {defaultValue=0.25, desc=Uses Core Wave, group=GrilledCheese, parameterName=GC_Stage3_WaveMultiplier, sortPriority=0}
  FunctionOutput(None=Add(A=Add(A=@n97, B=@n101), B=Multiply(A=@n86, B=@n102) {constB=0.25})) {outputName=Stage 3, sortPriority=3}
  n103 = VectorParameter() {group=GrilledCheese_FX, parameterName=GC_Stage1_CoreGlow_Color, defaultValue=<rgba 0.025 0.0229 0.0228 0.5>}
  n104 = Multiply(A=Multiply(A=@n16, B=@n103.RGB), B=@n17)
  n105 = ScalarParameter() {defaultValue=0.05, desc=uses Vortex C+D, group=GrilledCheese, parameterName=GC_Stage2_WhispsMultiplier, sortPriority=0}
  n106 = Multiply(A=LinearInterpolate(A=@n19, B=@n45, Alpha=@n73), B=@n105) {constB=0.1}
  n107 = ScalarParameter() {defaultValue=0.1, desc=uses Core Wave, group=GrilledCheese, parameterName=GC_Stage2_WaveMultiplier, sortPriority=0}
  FunctionOutput(None=Add(A=Add(A=@n104, B=@n106), B=Multiply(A=@n86, B=@n107) {constB=0.25})) {outputName=Stage 2, sortPriority=3}
  FunctionOutput(None=@n35.G) {outputName=Flame Noise 01, sortPriority=-1}
  LinearInterpolate(A=@n72, B=@n98, Alpha=@n69)
 ]}
