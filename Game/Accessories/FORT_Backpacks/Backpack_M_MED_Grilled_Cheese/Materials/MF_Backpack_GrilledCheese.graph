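# Material function for the Grilled Cheese backpack that layers animated energy and staged glow effects onto input material attributes.
# Triplanar-panned lightning gradients in pre-skinned space drive emissive energy, masked by FX texture channels.
# Stage parameters select pulsing fresnel and three socket-driven MF_Backpack_Sprite layers, masked by occlusion-texture RGB and scaled by time of day.
# Output base color is always the tint/cracks blend; the stage-aware base color blend is left unconnected.
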MaterialFunction /Game/Accessories/FORT_Backpacks/Backpack_M_MED_Grilled_Cheese/Materials/MF_Backpack_GrilledCheese
graph {nodes [
  n1 = Sine(None=Multiply(A=Time(), B=ScalarParameter() {defaultValue=0.1, group=GrilledCheese, parameterName=GC_FresnelPulse_Amount, sortPriority=103}) {constB=0.2}) {period=0.5}
  n2 = Clamp(
   None=Divide(A=Add(A=@n1))
   Min=ScalarParameter() {defaultValue=0.1, group=GrilledCheese, parameterName=GC_FresnelPulseMin, sortPriority=103}
   Max=ScalarParameter() {defaultValue=1, group=GrilledCheese, parameterName=GC_FresnelPulseMax, sortPriority=104}
  )
  n3 = TextureSampleParameter2D() {
   group=Grilled_Cheese_Textures
   parameterName=Skin FX
   texture=/Game/Accessories/FORT_Backpacks/Backpack_M_MED_Grilled_Cheese/Textures/T_M_MED_GrilledCheese_Backpack_FX
  }
  n4 = ChannelMaskParameter(None=@n3.RGBA) {group=GrilledCheese_Energy, maskChannel=Blue, parameterName=GC_Secondary Mask, sortPriority=-10, defaultValue=<rgba 0 0 1 0>}
  n5 = ChannelMaskParameter(None=@n3.RGBA) {group=GrilledCheese_Energy, parameterName=GC_Primary Mask, sortPriority=-10}
  n6 = Multiply(
   A=@n5.R
   B=AppendVector(
    A=ScalarParameter() {group=GrilledCheese_Energy, parameterName=GC_UVOffsetU}
    B=ScalarParameter() {group=GrilledCheese_Energy, parameterName=GC_UVOffsetV}
   )
  )
  n7 = ScalarParameter() {defaultValue=1, group=GrilledCheese_Energy, parameterName=GC_Emissive Tiling}
  n8 = PreSkinnedPosition()
  n9 = Divide(
   A=FunctionInput(Preview=VertexInterpolator(VS=@n8).PS) {bUsePreviewValueAsDefault=true, inputName=Local Position}
   B=ScalarParameter() {defaultValue=40, group=GrilledCheese_Energy, parameterName=GC_Triplanner_ConstantScaleDivisor}
  )
  n10 = Frac(None=Multiply(A=Time(), B=VectorParameter() {group=GrilledCheese_Energy, parameterName=GC_TriPlanner Speed, defaultValue=<rgba 0 0.2 0.2 1>}.RGB))
  n11 = ComponentMask(None=Add(A=Multiply(A=@n7, B=@n9), B=@n10)) {b=true, g=true}
  n12 = TextureObjectParameter() {
   group=Grilled_Cheese_Textures
   parameterName=GC_OcclusionTexture
   texture=/Game/Accessories/FORT_Backpacks/Backpack_M_MED_Grilled_Cheese/Textures/T_M_MED_GrilledCheese_Backpack_M2
  }
  n13 = TextureSample(UVs=TextureCoordinate(), Tex=@n12) {texture=&T_BlacksRGB}
  n14 = Multiply(A=@n13.A, B=ScalarParameter() {defaultValue=4, group=GrilledCheese_Energy, parameterName=GC_OffsetAmountInternal})
  n15 = TextureObjectParameter() {
   group=Grilled_Cheese_Textures
   parameterName=Panning Texture
   samplerType=SAMPLERTYPE_LinearColor
   texture=/Game/Characters/Player/Male/Medium/Bodies/M_MED_Lunch_Box/Textures/Lightning_Grad_M
  }
  n16 = TextureSample(UVs=Add(A=Add(A=@n6, B=@n11), B=@n14), Tex=@n15) {texture=&T_BlacksRGB}
  n17 = ComponentMask(None=Add(A=Multiply(A=@n7, B=@n9), B=@n10)) {b=true, r=true}
  n18 = TextureSample(UVs=Add(A=Add(A=@n6, B=@n17), B=@n14), Tex=@n15) {texture=&T_BlacksRGB}
  n19 = FunctionInput(Preview=VertexInterpolator(VS=PreSkinnedNormal()).PS) {bUsePreviewValueAsDefault=true, inputName=Local Normal}
  n20 = Abs(None=ComponentMask(None=@n19) {g=true})
  n21 = ScalarParameter() {defaultValue=1, group=GrilledCheese_Energy, parameterName=GC_TriPlanner_AxisBlendStrength}
  n22 = Saturate(None=ComponentMask(None=MaterialFunctionCall(In=@n20, Contrast=@n21) {materialFunction=&CheapContrast}.Result) {g=true})
  n23 = Add(A=Multiply(A=@n7, B=@n9), B=@n10)
  n24 = Add(A=Add(A=@n6, B=ComponentMask(None=@n23) {g=true, r=true}), B=@n14)
  n25 = ComponentMask(None=@n19) {b=true}
  n26 = ComponentMask(None=MaterialFunctionCall(In=Abs(None=@n25), Contrast=@n21) {materialFunction=&CheapContrast}.Result) {b=true}
  n27 = ComponentMask(None=LinearInterpolate(
    A=LinearInterpolate(A=@n16.RGB, B=@n18.RGB, Alpha=@n22)
    B=TextureSample(UVs=@n24, Tex=@n15) {texture=&T_BlacksRGB}.RGB
    Alpha=Saturate(None=@n26)
   )) {r=true}
  Add(A=LinearInterpolate(B=@n2, Alpha=@n4.R), B=@n27)
  n28 = GetMaterialAttributes(None=FunctionInput() {inputName=Material, inputType=FunctionInput_MaterialAttributes}) {attributeGetTypes [#1, #2]}
  n29 = GetMaterialAttributes(None=@n28.MaterialAttributes) {attributeGetTypes [#2]}
  n30 = Multiply(
   A=LinearInterpolate(
    A=VectorParameter() {group=GrilledCheese_Energy, parameterName=GC_Emissive Color A, defaultValue=<rgba 1 0.9995 0.1354 1>}.RGB
    B=VectorParameter() {group=GrilledCheese_Energy, parameterName=GC_Emissive Color B, defaultValue=<rgba 0 0.7217 1 1>}.RGB
    Alpha=@n5.R
   )
   B=LinearInterpolate(
    A=ScalarParameter() {defaultValue=0.01, group=GrilledCheese_Energy, parameterName=GC_Brightness_Min}
    B=ScalarParameter() {defaultValue=8, group=GrilledCheese_Energy, parameterName=GC_Brightness_Max}
    Alpha=@n27
   )
  )
  n31 = VectorParameter() {group=GrilledCheese_Energy, parameterName=GC_Emissive_Color_InnerShells, defaultValue=<rgba 1 0.7638 0.4464 1>}
  n32 = ChannelMaskParameter(None=@n3.RGBA) {group=GrilledCheese_Energy, maskChannel=Green, parameterName=GC_Inner Shells Mask, sortPriority=-10, defaultValue=<rgba 0 1 0 0>}
  n33 = LinearInterpolate(A=@n30, B=Add(A=@n30, B=@n31.RGB), Alpha=@n32.R)
  n34 = Multiply(A=@n27, B=@n29.BaseColor)
  n35 = Power(Base=Multiply(A=@n13.A) {constB=0.05}) {constExponent=0.7}
  n36 = VectorParameter() {group=GrilledCheese, parameterName=sphereASocket, sortPriority=300, defaultValue=<rgba 6 6 0 0>}
  n37 = MaterialFunctionCall!(Warp Texture=@n13.A, In Socket UVs=@n36.RGB) {materialFunction=&/MF_Backpack_Sprite}
  n38 = Multiply(A=@n37.Flame Noise 01, B=@n37.Flame Noise 01)
  n39 = Multiply(A=@n32.R) {constB=0.75}
  n40 = ScalarParameter() {group=GrilledCheese, parameterName=GC_UseStage2, sortPriority=-1}
  n41 = ScalarParameter() {group=GrilledCheese, parameterName=GC_UseStage3, sortPriority=-1}
  n42 = LinearInterpolate(A=LinearInterpolate(A=Constant(), B=@n39, Alpha=@n40), B=@n39, Alpha=@n41)
  n43 = ScalarParameter() {group=GrilledCheese, parameterName=GC_UseStage4, sortPriority=-1}
  n44 = Add(A=@n4.R, B=Multiply(A=Add(A=@n35, B=@n38), B=LinearInterpolate(A=@n42, B=@n39, Alpha=@n43)))
  n45 = GetMaterialAttributes(None=BlendMaterialAttributes(
    A=@n28.MaterialAttributes
    B=SetMaterialAttributes!(MaterialAttributes=@n29.MaterialAttributes, EmissiveColor=@n33, BaseColor=@n34) {attributeSetTypes [#1, #2]}
    Alpha=Saturate(None=@n44)
   )) {attributeGetTypes [#2, #1]}
  n46 = ScalarParameter() {group=GrilledCheese, parameterName=GC_UseStage1, sortPriority=-1}
  n47 = Constant3Vector()
  n48 = LinearInterpolate(A=LinearInterpolate(A=@n45.EmissiveColor, B=@n47, Alpha=@n46), B=@n45.EmissiveColor, Alpha=@n40)
  n49 = VectorParameter() {group=GrilledCheese, parameterName=GC_Stage1_Fresnel_Color, sortPriority=0, defaultValue=<rgba 0.6637 0.6259 0.5254 1>}
  n50 = ChannelMaskParameter(None=@n3.RGBA) {group=GrilledCheese_Energy, maskChannel=Alpha, parameterName=GC_Fresnel Edge Mask, sortPriority=-10, defaultValue=<rgba 0 0 0 1>}
  n51 = Fresnel(ExponentIn=ScalarParameter() {defaultValue=2, group=GrilledCheese, parameterName=GC_FresnelExponent}) {baseReflectFraction=0}
  n52 = ScalarParameter() {defaultValue=2, group=GrilledCheese, parameterName=GC_FresnelPower}
  n53 = Saturate(None=Add(A=Multiply(A=@n51, B=@n38), B=Power(Base=@n51, Exp=@n52)))
  n54 = Multiply(
   A=ScalarParameter() {defaultValue=0.25, group=GrilledCheese, parameterName=GC_Stage1_FresnelIntensity, sortPriority=0}
   B=LinearInterpolate(A=Constant3Vector(), B=@n49.RGB, Alpha=Multiply(A=@n50.R, B=@n53))
  )
  n55 = VectorParameter() {group=GrilledCheese, parameterName=sphereBSocket, sortPriority=300, defaultValue=<rgba 6 6 0 0>}
  n56 = MaterialFunctionCall(Warp Texture=@n13.A, In Socket UVs=@n55.RGB) {materialFunction=&MF_Backpack_Sprite}
  n57 = VectorParameter() {group=GrilledCheese, parameterName=sphereCSocket, sortPriority=300, defaultValue=<rgba 6 6 0 0>}
  n58 = MaterialFunctionCall(Warp Texture=@n13.A, In Socket UVs=@n57.RGB) {materialFunction=&MF_Backpack_Sprite}
  n59 = Add(A=Add(A=Multiply(A=@n13.R, B=@n37.Stage 2), B=Multiply(A=@n13.G, B=@n56.Stage 2)), B=Multiply(A=@n13.B, B=@n58.Stage 2))
  n60 = Multiply(A=@n59) {constB=0.75}
  n61 = VectorParameter() {group=GrilledCheese, parameterName=GC_Fresnel_ColorA, defaultValue=<rgba 1 0.4534 0 0>}
  n62 = VectorParameter() {group=GrilledCheese, parameterName=GC_Fresnel_ColorB, defaultValue=<rgba 10 2.2669 0 0>}
  n63 = Multiply(
   A=Multiply(
    A=LinearInterpolate(A=@n61.RGB, B=@n62.RGB, Alpha=@n2)
    B=ScalarParameter() {defaultValue=1, group=GrilledCheese, parameterName=GC_FresnelBrightness}
   )
   B=@n53
  )
  n64 = Add(A=Multiply(A=@n13.R, B=@n37.Stage 3), B=Multiply(A=@n13.G, B=@n56.Stage 3))
  n65 = Multiply(A=@n13.B, B=@n58.Stage 3)
  n66 = LinearInterpolate(A=LinearInterpolate(A=Add(A=@n54, B=@n60), B=Add(A=@n63, B=@n59), Alpha=@n40), B=Add(A=@n63, B=Add(A=@n64, B=@n65)), Alpha=@n41)
  n67 = Multiply(A=@n13.R, B=@n37.Stage 4)
  n68 = Multiply(A=@n13.G, B=@n56.Stage 4)
  n69 = Add(A=@n63, B=Add(A=Add(A=@n67, B=@n68), B=Multiply(A=@n13.B, B=@n58.Stage 4)))
  n70 = VectorParameter() {group=GrilledCheese, parameterName=GC_TOD_EMissiveMultiplier_Morning, sortPriority=200, defaultValue=<rgba 1 1 1 1>}
  n71 = VectorParameter() {group=GrilledCheese, parameterName=GC_TOD_EMissiveMultiplier_Day, sortPriority=200, defaultValue=<rgba 1 1 1 1>}
  n72 = VectorParameter() {group=GrilledCheese, parameterName=GC_TOD_EMissiveMultiplier_Evening, sortPriority=200, defaultValue=<rgba 1 1 1 1>}
  n73 = VectorParameter() {group=GrilledCheese, parameterName=GC_TOD_EMissiveMultiplier_Night, sortPriority=200, defaultValue=<rgba 0.25 0.25 0.25 1>}
  n74 = Add(
   A=LinearInterpolate(A=LinearInterpolate(A=@n28.EmissiveColor, B=@n47, Alpha=@n46), B=@n28.EmissiveColor, Alpha=@n40)
   B=Multiply(
    A=LinearInterpolate(A=@n66, B=@n69, Alpha=@n43)
    B=MaterialFunctionCall(Color_Morning=@n70.RGB, Color_Day=@n71.RGB, Color_Evening=@n72.RGB, Color_Night=@n73.RGB) {materialFunction=&Color_Time_of_Day}.Result
   )
  )
  n75 = VectorParameter() {group=GrilledCheese, parameterName=GC_BaseColor_Tint, sortPriority=0, defaultValue=<rgba 0.0212 0.0203 0.0212 1>}
  n76 = VectorParameter() {group=GrilledCheese, parameterName=GC_BaseColor_Cracks, sortPriority=0, defaultValue=<rgba 0.01 0.0096 0.01 1>}
  n77 = LinearInterpolate(A=@n75.RGB, B=@n76.RGB, Alpha=@n4.R)
  FunctionOutput(None=SetMaterialAttributes(MaterialAttributes=@n45.MaterialAttributes, EmissiveColor=Add(A=@n48, B=@n74), BaseColor=@n77) {attributeSetTypes [#1, #2]})
  LinearInterpolate(A=LinearInterpolate(A=@n45.BaseColor, B=@n77, Alpha=@n46), B=@n45.BaseColor, Alpha=@n40)
  n78 = ActorPositionWS()
  n79 = ScalarParameter() {group=GrilledCheese, parameterName=GC_ForRendering, sortPriority=-300}
  LinearInterpolate(A=@n36.RGB, B=@n78, Alpha=@n79)
  LinearInterpolate(A=@n55.RGB, B=@n78, Alpha=@n79)
  LinearInterpolate(A=@n57.RGB, B=@n78, Alpha=@n79)
 ]}
