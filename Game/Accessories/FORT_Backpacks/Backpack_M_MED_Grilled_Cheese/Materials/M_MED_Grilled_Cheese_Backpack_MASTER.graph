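# Master material for the Grilled Cheese backpack accessory, a Default Lit skeletal-mesh material that uses material attributes and supports mesh deformers.
# The shared character PreFX function feeds the backpack-specific MF_Backpack_GrilledCheese function, including pre-skinned local position.
# That result passes through MF_MetalsLUT into the shared character PostFX function, which also takes inputs from PreFX.
# The graph is only this function chain, with surface detail delegated to the called functions.
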Material /Game/Accessories/FORT_Backpacks/Backpack_M_MED_Grilled_Cheese/Materials/M_MED_Grilled_Cheese_Backpack_MASTER
graph {
 properties {
  bUseMaterialAttributes=true
  bUsedWithMeshDeformer=true
  bUsedWithSkeletalMesh=true
  pixelDepthOffsetMode=PDOM_Legacy
  usedShadingModels=Default Lit
 }
 nodes [
  n1 = MaterialFunctionCall() {materialFunction=&BaseFN_CharacterFunction_PreFX}
  MaterialFunctionCall(
   AO=@n1.AO
   Distance Blend=@n1.AO
   Fresnel=@n1.AO
   Material=MaterialFunctionCall(Materials=MaterialFunctionCall!(Local Position=@n1.Pre skinned Local Position, Local Normal, Material=@n1.Pre skinned Local Position) {materialFunction=&/MF_Backpack_GrilledCheese}.Result) {materialFunction=&MF_MetalsLUT}.Result
  ) {materialFunction=&BaseFN_CharacterFunction_PostFX}
 ]
}
