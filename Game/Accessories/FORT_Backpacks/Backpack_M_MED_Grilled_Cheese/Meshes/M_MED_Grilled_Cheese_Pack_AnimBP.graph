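# Anim Blueprint for the Grilled Cheese backpack, animating three floating, cracking spheres that react to the player's kill count.
# Spring AnimDynamics feed additive bone transforms for spin, sine bobbing and crack offsets; an Actor-space dynamics chain is left unconnected.
# Kills, from a stat watcher or front-end event, trigger base-spin surges, with bigger surges and randomized crack spread at listed counts.
# Post-evaluation pushes sphere socket locations into material parameters.
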AnimBlueprint /Game/Accessories/FORT_Backpacks/Backpack_M_MED_Grilled_Cheese/Meshes/M_MED_Grilled_Cheese_Pack_AnimBP
target_skeleton=&/M_MED_Grilled_Cheese_Pack_Skeleton
graphs [
 AnimGraph {nodes [
   n1 = Anim Dynamics - Bone: sphereA(Alpha=0.5, ExternalForce="0,0,0") {
    angularDampingOverride=0.999
    angularSpringConstant=150
    bAngularSpring=true
    bGravityOverrideInSimSpace=true
    bLinearSpring=true
    bOverrideAngularDamping=true
    bOverrideLinearDamping=true
    bUseGravityOverride=true
    gravityScale=0
    linearDampingOverride=0.995
    linearSpringConstant=50
    simulationSpace=Actor
    alphaBoolBlend {blendOption=Linear}
    alphaScaleBiasClamp {bInterpResult=true, interpSpeedDecreasing=3, interpSpeedIncreasing=5}
    boundBone {boneName=sphereA}
    componentAppliedLinearAccClamp=<6000 6000 6000>
    componentLinearAccScale=<0.05 0.05 0.05>
    componentLinearVelScale=<0.8 0.8 0.6>
    gravityOverride=<0.25 0.25 0.25>
    physicsBodyDefinitions [{
      sphereCollisionRadius=0
      boundBone {boneName=sphereA}
      boxExtents=<50 50 50>
      constraintSetup {angularLimitsMax=<0 0 0>, angularLimitsMin=<0 0 0>, angularTarget=<0 0 0>, linearAxesMax=<1 0.5 1>, linearAxesMin=<-1 -0.5 -1>}
      localJointOffset=<0 -3 0>
     }]
    relativeSpaceBone {boneName=root}
   }
   n2 = Anim Dynamics - Bone: sphereB(ComponentPose=@n1, Alpha=0.35, ExternalForce="0,0,0") {
    angularDampingOverride=0.999
    angularSpringConstant=150
    bAngularSpring=true
    bGravityOverrideInSimSpace=true
    bLinearSpring=true
    bOverrideAngularDamping=true
    bOverrideLinearDamping=true
    bUseGravityOverride=true
    gravityScale=0
    linearDampingOverride=0.995
    linearSpringConstant=50
    simulationSpace=Actor
    alphaBoolBlend {blendOption=Linear}
    alphaScaleBiasClamp {bInterpResult=true, interpSpeedDecreasing=3, interpSpeedIncreasing=5}
    boundBone {boneName=sphereB}
    componentAppliedLinearAccClamp=<6000 6000 6000>
    componentLinearAccScale=<0.05 0.05 0.05>
    componentLinearVelScale=<0.8 0.8 0.6>
    gravityOverride=<0.25 0.25 0.25>
    physicsBodyDefinitions [{
      sphereCollisionRadius=0
      boundBone {boneName=sphereB}
      boxExtents=<50 50 50>
      constraintSetup {angularLimitsMax=<0 0 0>, angularLimitsMin=<0 0 0>, angularTarget=<0 0 0>, linearAxesMax=<1 0.5 1>, linearAxesMin=<-1 -0.5 -1>}
      localJointOffset=<0 -3 0>
     }]
    relativeSpaceBone {boneName=root}
   }
   Anim Dynamics - Bone: sphereC(ComponentPose=@n2, Alpha=0.2, ExternalForce="0,0,0") {
    angularDampingOverride=0.999
    angularSpringConstant=150
    bAngularSpring=true
    bGravityOverrideInSimSpace=true
    bLinearSpring=true
    bOverrideAngularDamping=true
    bOverrideLinearDamping=true
    bUseGravityOverride=true
    gravityScale=0
    linearDampingOverride=0.995
    linearSpringConstant=50
    simulationSpace=Actor
    alphaBoolBlend {blendOption=Linear}
    alphaScaleBiasClamp {bInterpResult=true, interpSpeedDecreasing=3, interpSpeedIncreasing=5}
    boundBone {boneName=sphereC}
    componentAppliedLinearAccClamp=<6000 6000 6000>
    componentLinearAccScale=<0.05 0.05 0.05>
    componentLinearVelScale=<0.8 0.8 0.6>
    gravityOverride=<0.25 0.25 0.25>
    physicsBodyDefinitions [{
      sphereCollisionRadius=0
      boundBone {boneName=sphereC}
      boxExtents=<50 50 50>
      constraintSetup {angularLimitsMax=<0 0 0>, angularLimitsMin=<0 0 0>, angularTarget=<0 0 0>, linearAxesMax=<1 0 1>, linearAxesMin=<-1 -0.5 -1>}
      localJointOffset=<0 -3 0>
     }]
    relativeSpaceBone {boneName=root}
   }
   n3 = Anim Dynamics - Bone: sphereA(ComponentPose=Local To Component(LocalPose=Local Space Ref Pose()), Alpha=1, ExternalForce="0,0,0") {
    angularDampingOverride=0.999
    angularSpringConstant=150
    bAngularSpring=true
    bGravityOverrideInSimSpace=true
    bLinearSpring=true
    bOverrideAngularDamping=true
    bOverrideLinearDamping=true
    bUseGravityOverride=true
    gravityScale=0
    linearDampingOverride=0.995
    linearSpringConstant=50
    simulationSpace=RootRelative
    alphaBoolBlend {blendOption=Linear}
    alphaScaleBiasClamp {interpSpeedDecreasing=3, interpSpeedIncreasing=5}
    boundBone {boneName=sphereA}
    componentAppliedLinearAccClamp=<6000 6000 6000>
    componentLinearAccScale=<0.05 0.05 0.05>
    componentLinearVelScale=<0.8 0.8 0.6>
    gravityOverride=<0.25 0.25 0.25>
    physicsBodyDefinitions [{
      sphereCollisionRadius=0
      boundBone {boneName=sphereA}
      boxExtents=<50 50 50>
      constraintSetup {angularLimitsMax=<0 0 0>, angularLimitsMin=<0 0 0>, angularTarget=<0 0 0>, linearAxesMax=<1 0.5 1>, linearAxesMin=<-1 -0.5 -1>}
      localJointOffset=<0 -3 0>
     }]
    relativeSpaceBone {boneName=root}
   }
   n4 = Anim Dynamics - Bone: sphereB(ComponentPose=@n3, Alpha=0.75, ExternalForce="0,0,0") {
    angularDampingOverride=0.999
    angularSpringConstant=150
    bAngularSpring=true
    bGravityOverrideInSimSpace=true
    bLinearSpring=true
    bOverrideAngularDamping=true
    bOverrideLinearDamping=true
    bUseGravityOverride=true
    gravityScale=0
    linearDampingOverride=0.995
    linearSpringConstant=50
    simulationSpace=RootRelative
    alphaBoolBlend {blendOption=Linear}
    alphaScaleBiasClamp {interpSpeedDecreasing=3, interpSpeedIncreasing=5}
    boundBone {boneName=sphereB}
    componentAppliedLinearAccClamp=<6000 6000 6000>
    componentLinearAccScale=<0.05 0.05 0.05>
    componentLinearVelScale=<0.8 0.8 0.6>
    gravityOverride=<0.25 0.25 0.25>
    physicsBodyDefinitions [{
      sphereCollisionRadius=0
      boundBone {boneName=sphereB}
      boxExtents=<50 50 50>
      constraintSetup {angularLimitsMax=<0 0 0>, angularLimitsMin=<0 0 0>, angularTarget=<0 0 0>, linearAxesMax=<1 0.5 1>, linearAxesMin=<-1 -0.5 -1>}
      localJointOffset=<0 -3 0>
     }]
    relativeSpaceBone {boneName=root}
   }
   n5 = Anim Dynamics - Bone: sphereC(ComponentPose=@n4, Alpha=0.5, ExternalForce="0,0,0") {
    angularDampingOverride=0.999
    angularSpringConstant=150
    bAngularSpring=true
    bGravityOverrideInSimSpace=true
    bLinearSpring=true
    bOverrideAngularDamping=true
    bOverrideLinearDamping=true
    bUseGravityOverride=true
    gravityScale=0
    linearDampingOverride=0.995
    linearSpringConstant=50
    simulationSpace=RootRelative
    alphaBoolBlend {blendOption=Linear}
    alphaScaleBiasClamp {interpSpeedDecreasing=3, interpSpeedIncreasing=5}
    boundBone {boneName=sphereC}
    componentAppliedLinearAccClamp=<6000 6000 6000>
    componentLinearAccScale=<0.05 0.05 0.05>
    componentLinearVelScale=<0.8 0.8 0.6>
    gravityOverride=<0.25 0.25 0.25>
    physicsBodyDefinitions [{
      sphereCollisionRadius=0
      boundBone {boneName=sphereC}
      boxExtents=<50 50 50>
      constraintSetup {angularLimitsMax=<0 0 0>, angularLimitsMin=<0 0 0>, angularTarget=<0 0 0>, linearAxesMax=<1 0 1>, linearAxesMin=<-1 -0.5 -1>}
      localJointOffset=<0 -3 0>
     }]
    relativeSpaceBone {boneName=root}
   }
   n6 = "Transform (Modify) Bone - Bone: sphereBase"(ComponentPose=@n5, Alpha=1, Translation="0,0,0", Rotation=Get Spinning_Rotator_00() -> (Spinning_Rotator_00:Rotator), Scale="1,1,1") {
    rotationMode=BMM_Additive
    rotationSpace=BCS_BoneSpace
    translationMode=BMM_Additive
    translationSpace=BCS_BoneSpace
    alphaBoolBlend {blendOption=Linear}
    alphaScaleBiasClamp {interpSpeedDecreasing=4, interpSpeedIncreasing=6}
    boneToModify {boneName=sphereBase}
   }
   n7 = "Transform (Modify) Bone - Bone: sphereA"(
    ComponentPose=@n6
    Alpha=1
    Translation=Get Cube_Float_A() -> (Cube_Float_A:Vector)
    Rotation=Get Spinning_Rotator_01() -> (Spinning_Rotator_01:Rotator)
    Scale="1,1,1"
   ) {
    rotationMode=BMM_Additive
    rotationSpace=BCS_BoneSpace
    translationMode=BMM_Additive
    alphaBoolBlend {blendOption=Linear}
    alphaScaleBiasClamp {bInterpResult=true, interpSpeedDecreasing=4, interpSpeedIncreasing=6}
    boneToModify {boneName=sphereA}
   }
   n8 = "Transform (Modify) Bone - Bone: sphereB"(
    ComponentPose=@n7
    Alpha=1
    Translation=Get Cube_Float_B() -> (Cube_Float_B:Vector)
    Rotation=Get Spinning_Rotator_02() -> (Spinning_Rotator_02:Rotator)
    Scale="1,1,1"
   ) {
    rotationMode=BMM_Additive
    rotationSpace=BCS_BoneSpace
    translationMode=BMM_Additive
    alphaBoolBlend {blendOption=Linear}
    alphaScaleBiasClamp {bInterpResult=true, interpSpeedDecreasing=4, interpSpeedIncreasing=6}
    boneToModify {boneName=sphereB}
   }
   n9 = "Transform (Modify) Bone - Bone: sphereC"(
    ComponentPose=@n8
    Alpha=1
    Translation=Get Cube_Float_A()
    Rotation=Get Spinning_Rotator_03() -> (Spinning_Rotator_03:Rotator)
    Scale="1,1,1"
   ) {
    rotationMode=BMM_Additive
    rotationSpace=BCS_BoneSpace
    translationMode=BMM_Additive
    alphaBoolBlend {blendOption=Linear}
    alphaScaleBiasClamp {bInterpResult=true, interpSpeedDecreasing=4, interpSpeedIncreasing=6}
    boneToModify {boneName=sphereC}
   }
   n10 = Get crackRot() -> (crackRot:Rotator)
   n11 = Get crackTran() -> (crackTran:Vector)
   n12 = "Transform (Modify) Bone - Bone: sphereA_crackB"(ComponentPose=@n9, Alpha=1, Translation=@n11, Rotation=@n10, Scale="1,1,1") {
    rotationMode=BMM_Additive
    rotationSpace=BCS_BoneSpace
    translationMode=BMM_Additive
    translationSpace=BCS_BoneSpace
    alphaBoolBlend {blendOption=Linear}
    alphaScaleBiasClamp {bInterpResult=true, interpSpeedDecreasing=4, interpSpeedIncreasing=6}
    boneToModify {boneName=sphereA_crackB}
   }
   n13 = "Transform (Modify) Bone - Bone: sphereA_crackC"(ComponentPose=@n12, Alpha=1, Translation=@n11, Rotation=@n10, Scale="1,1,1") {
    rotationMode=BMM_Additive
    rotationSpace=BCS_BoneSpace
    translationMode=BMM_Additive
    translationSpace=BCS_BoneSpace
    alphaBoolBlend {blendOption=Linear}
    alphaScaleBiasClamp {bInterpResult=true, interpSpeedDecreasing=4, interpSpeedIncreasing=6}
    boneToModify {boneName=sphereA_crackC}
   }
   n14 = "Transform (Modify) Bone - Bone: sphereA_crackD"(ComponentPose=@n13, Alpha=1, Translation=@n11, Rotation=@n10, Scale="1,1,1") {
    rotationMode=BMM_Additive
    rotationSpace=BCS_BoneSpace
    translationMode=BMM_Additive
    translationSpace=BCS_BoneSpace
    alphaBoolBlend {blendOption=Linear}
    alphaScaleBiasClamp {bInterpResult=true, interpSpeedDecreasing=4, interpSpeedIncreasing=6}
    boneToModify {boneName=sphereA_crackD}
   }
   n15 = "Transform (Modify) Bone - Bone: sphereA_crackA"(ComponentPose=@n14, Alpha=1, Translation=@n11, Rotation=@n10, Scale="1,1,1") {
    rotationMode=BMM_Additive
    rotationSpace=BCS_BoneSpace
    translationMode=BMM_Additive
    translationSpace=BCS_BoneSpace
    alphaBoolBlend {blendOption=Linear}
    alphaScaleBiasClamp {bInterpResult=true, interpSpeedDecreasing=4, interpSpeedIncreasing=6}
    boneToModify {boneName=sphereA_crackA}
   }
   n16 = Get crackRot()
   n17 = Get crackTran()
   n18 = "Transform (Modify) Bone - Bone: sphereB_crackA"(ComponentPose=@n15, Alpha=0.7, Translation=@n17, Rotation=@n16, Scale="1,1,1") {
    rotationMode=BMM_Additive
    rotationSpace=BCS_BoneSpace
    translationMode=BMM_Additive
    translationSpace=BCS_BoneSpace
    alphaBoolBlend {blendOption=Linear}
    alphaScaleBiasClamp {bInterpResult=true, interpSpeedDecreasing=4, interpSpeedIncreasing=6}
    boneToModify {boneName=sphereB_crackA}
   }
   n19 = "Transform (Modify) Bone - Bone: sphereB_crackB"(ComponentPose=@n18, Alpha=0.7, Translation=@n17, Rotation=@n16, Scale="1,1,1") {
    rotationMode=BMM_Additive
    rotationSpace=BCS_BoneSpace
    translationMode=BMM_Additive
    translationSpace=BCS_BoneSpace
    alphaBoolBlend {blendOption=Linear}
    alphaScaleBiasClamp {bInterpResult=true, interpSpeedDecreasing=4, interpSpeedIncreasing=6}
    boneToModify {boneName=sphereB_crackB}
   }
   n20 = "Transform (Modify) Bone - Bone: sphereB_crackC"(ComponentPose=@n19, Alpha=0.7, Translation=@n17, Rotation=@n16, Scale="1,1,1") {
    rotationMode=BMM_Additive
    rotationSpace=BCS_BoneSpace
    translationMode=BMM_Additive
    translationSpace=BCS_BoneSpace
    alphaBoolBlend {blendOption=Linear}
    alphaScaleBiasClamp {bInterpResult=true, interpSpeedDecreasing=4, interpSpeedIncreasing=6}
    boneToModify {boneName=sphereB_crackC}
   }
   n21 = "Transform (Modify) Bone - Bone: sphereB_crackD"(ComponentPose=@n20, Alpha=0.7, Translation=@n17, Rotation=@n16, Scale="1,1,1") {
    rotationMode=BMM_Additive
    rotationSpace=BCS_BoneSpace
    translationMode=BMM_Additive
    translationSpace=BCS_BoneSpace
    alphaBoolBlend {blendOption=Linear}
    alphaScaleBiasClamp {bInterpResult=true, interpSpeedDecreasing=4, interpSpeedIncreasing=6}
    boneToModify {boneName=sphereB_crackD}
   }
   n22 = "Transform (Modify) Bone - Bone: sphereB_crackE"(ComponentPose=@n21, Alpha=0.7, Translation=@n17, Rotation=@n16, Scale="1,1,1") {
    rotationMode=BMM_Additive
    rotationSpace=BCS_BoneSpace
    translationMode=BMM_Additive
    translationSpace=BCS_BoneSpace
    alphaBoolBlend {blendOption=Linear}
    alphaScaleBiasClamp {bInterpResult=true, interpSpeedDecreasing=4, interpSpeedIncreasing=6}
    boneToModify {boneName=sphereB_crackE}
   }
   n23 = Get crackRot()
   n24 = Get crackTran()
   n25 = "Transform (Modify) Bone - Bone: sphereC_crackA"(ComponentPose=@n22, Alpha=0.5, Translation=@n24, Rotation=@n23, Scale="1,1,1") {
    rotationMode=BMM_Additive
    rotationSpace=BCS_BoneSpace
    translationMode=BMM_Additive
    translationSpace=BCS_BoneSpace
    alphaBoolBlend {blendOption=Linear}
    alphaScaleBiasClamp {bInterpResult=true, interpSpeedDecreasing=4, interpSpeedIncreasing=6}
    boneToModify {boneName=sphereC_crackA}
   }
   n26 = "Transform (Modify) Bone - Bone: sphereC_crackB"(ComponentPose=@n25, Alpha=0.5, Translation=@n24, Rotation=@n23, Scale="1,1,1") {
    rotationMode=BMM_Additive
    rotationSpace=BCS_BoneSpace
    translationMode=BMM_Additive
    translationSpace=BCS_BoneSpace
    alphaBoolBlend {blendOption=Linear}
    alphaScaleBiasClamp {bInterpResult=true, interpSpeedDecreasing=4, interpSpeedIncreasing=6}
    boneToModify {boneName=sphereC_crackB}
   }
   Output Pose(Result=Component To Local(ComponentPose=@n26)) {layerGroup=None}
  ]}
 EventGraph {nodes [
   n1 = UpdateFEKillCountEvent!() -> (OutputDelegate, then, NewParam:double=~) {is=title K2Node_CustomEvent}
   > Branch(Condition=Get bIsInFrontEnd())
   > n2 = Set killCount(killCount:float=@n1.NewParam) -> (Output_Get:float)
   Set ActiveKill(execute=[@n8.then, @n2.then], ActiveKill:bool=true) -> (Output_Get:bool)
   Event BlueprintInitializeAnimation()
   > Set Skel_Mesh(Skel_Mesh=GetOwnerSkeletalMeshComponent())
   > n3 = Cast To FortPlayerPawn(Object=TryGetFortPawn())
   > Set Pawn(Pawn:Fort Player Pawn Object Reference=@n3) -> (Output_Get:Fort Player Pawn Object Reference)
   n4 = Event BlueprintUpdateAnimation()
   > n5 = Sequence!(execute) -> (then_0, then_1, then_2:Exec, then_3:Exec, then_4:Exec)
   > UpdateSpinSettings(self:Custom Character Part Anim Instance Grilled Cheese Pack Object Reference)
   UpdateCracks(execute=@n5.then_1, self:Custom Character Part Anim Instance Grilled Cheese Pack Object Reference)
   UpdateCrackSpinLogic!(execute=@n5.then_2, self:Custom Character Part Anim Instance Grilled Cheese Pack Object Reference, DeltaTime:float=@n4.DeltaTimeX) -> (then)
   UpdateFloatCubes(execute=@n5.then_3, self:Custom Character Part Anim Instance Grilled Cheese Pack Object Reference)
   Event OnCharacterCustomizationLoaded()
   > n6 = Branch(Condition=Get bIsInFrontEnd())
   n7 = WatchCosmeticStat!(
    execute:Exec=@n6.else
    Pawn:Fort Player Pawn Object Reference=Get Pawn() -> (Pawn:Fort Player Pawn Object Reference)
    StatTag:Fort Stat Manager Tag Structure=(TagName="StatManager.AthenaKills")
    DefaultValue:int=0
    bBroadcastInitialValue:bool=false
    InitialValueBroadcastDelay:float=-1
    ChangeValueBroadcastDelay:float=-1
    bConsumeWaitingNotifications:bool=true
   ) -> (then:Exec, OnStatChanged:Exec, StatName:Name, StatValue:int) {is=title K2Node_AsyncAction}
   n8 = Set killCount(execute=@n7.OnStatChanged, killCount="To Float (Integer)"(InInt=@n7.StatValue))
   Event BlueprintPostEvaluateAnimation() {is=title K2Node_Event}
   n9 = GetOwningComponent(self=Self-Reference())
   > Is Valid(InputObject=@n9)
   > Delay(Duration=0)
   > SetVectorParameterValueOnMaterials!(
    execute
    self:Mesh Component Object Reference=@n9
    ParameterName:Name=sphereASocket
    ParameterValue:Vector=GetSocketLocation!(self:Scene Component Object Reference=@n9, InSocketName:Name=sphereASocket) -> (ReturnValue:Vector)
   ) -> (then)
   > SetVectorParameterValueOnMaterials(self=@n9, ParameterName=sphereBSocket, ParameterValue=GetSocketLocation(self=@n9, InSocketName=sphereBSocket))
   > SetVectorParameterValueOnMaterials(self=@n9, ParameterName=sphereCSocket, ParameterValue=GetSocketLocation(self=@n9, InSocketName=sphereCSocket))
   n10 = Sequence()
   n11 = GetRealTimeSeconds()
   n12 = "Sin (Radians)"(A=float * float(A=float * float(A=1, B=1), B=@n11))
   n13 = "Sin (Radians)"(A=float * float(A=float * float(A=1, B=1.4), B=@n11))
   n14 = "Sin (Radians)"(A=float * float(A=float * float(A=1, B=1.2), B=@n11))
   > Set Cube_Float_A(Cube_Float_A:Vector=MakeVector(X=float * float(A=@n14, B=0.3), Y=float * float(A=@n12, B=0.56), Z=float * float(A=@n13, B=0.22))) -> (Output_Get:Vector)
   n15 = GetGameTimeInSeconds()
   n16 = "Cos (Radians)"!(A:double=float * float(A=float * float(A=1, B=1.5), B=@n15)) -> (ReturnValue:double)
   n17 = "Cos (Radians)"(A=float * float(A=float * float(A=1, B=1.15), B=@n15))
   n18 = "Cos (Radians)"(A=float * float(A=float * float(A=1, B=1.15), B=@n15))
   Set Cube_Float_B(
    execute=@n10.then_1
    Cube_Float_B:Vector=MakeVector(X=float * float(A=@n16, B=0.2), Y=float * float(A=@n17, B=0.725), Z=float * float(A=@n18, B=0.3))
   ) -> (Output_Get:Vector)
   n19 = Sequence()
   > n20 = Branch(Condition=Contains Item(
     TargetArray:"Array of Float (single-precision)s"=Get killCrackArray() -> (killCrackArray:"Array of Float (single-precision)s")
     ItemToFind:"Float (single-precision) (by ref)"=Get killCount() -> (killCount:float)
    ))
   > Set interpSpeed(interpSpeed:float=15) -> (Output_Get:float)
   > Set onCrackSpinSpeed(onCrackSpinSpeed:float=20) -> (Output_Get:float)
   > Set isCrackCount(isCrackCount:bool=true) -> (Output_Get:bool)
   > n21 = Branch(Condition="float > float"(A=Get killCount(), B=1) {is=title K2Node_CallFunction})
   n22 = Get killCount()
   > Set crackCount(crackCount:float=float + float(A=float / float(A=@n22, B=3) {is=title K2Node_CallFunction}, B=1) {is=title K2Node_CommutativeAssociativeBinaryOperator}) -> (Output_Get:float)
   Set crackCount(execute=@n21.else, crackCount=Get killCount())
   Set interpSpeed(execute=@n20.else, interpSpeed=6)
   > Set onCrackSpinSpeed(onCrackSpinSpeed=8)
   > Set isCrackCount(isCrackCount=false)
   n23 = Branch(execute=@n19.then_1, Condition=Get ActiveKill() -> (ActiveKill:bool))
   > Branch(Condition=Get isCrackCount() -> (isCrackCount:bool))
   > n24 = Branch(Condition="float > float"(A=Get crackCount() -> (crackCount:float), B=0.1))
   n25 = Get crackCount()
   > Set crackTran(crackTran:Vector=MakeVector(Y=0, Z=float * float(A=@n25, B=0.25))) -> (Output_Get:Vector)
   n26 = RandomFloatInRange(Min=0, Max=5)
   > Set crackRot(crackRot:Rotator=MakeRotator(Roll=@n26, Pitch=@n26, Yaw=@n26)) -> (Output_Get:Rotator)
   Set crackTran(execute=@n24.else, crackTran="0, 0, 0")
   > Set crackRot(crackRot="0, 0, 0")
   n27 = float * float(A=Get crackCount(), B=0.005)
   n28 = Get Cube_Float_A()
   Set crackTran(
    execute=@n23.else
    crackTran=vector + vector(A:Vector=vector * float(A=@n28, B=@n27), B:Vector=Get crackTran()) -> (ReturnValue:Vector) {is=title K2Node_CommutativeAssociativeBinaryOperator}
   )
   n29 = Sequence!(execute) -> (then_0, then_1, then_2)
   n30 = spinFNC_deprecated!(execute=@n29.then_2, self, speed:double=-14, initVal:double=0) -> (then, output:double)
   > Set Spinning_Rotator_03(Spinning_Rotator_03:Rotator=MakeRotator(Roll=@n30, Pitch=@n30, Yaw=@n30)) -> (Output_Get:Rotator)
   n31 = spinFNC_deprecated(execute=@n29.then_0, speed=-10, initVal=0)
   > Set Spinning_Rotator_01(Spinning_Rotator_01:Rotator=MakeRotator(Roll=@n31, Pitch=@n31, Yaw=@n31)) -> (Output_Get:Rotator)
   n32 = spinFNC_deprecated(execute=@n29.then_1, speed=12, initVal=0)
   > Set Spinning_Rotator_02(Spinning_Rotator_02:Rotator=MakeRotator(Roll=@n32, Pitch=@n32, Yaw=@n32)) -> (Output_Get:Rotator)
   n33 = Sequence!(execute) -> (then_0, then_1)
   n34 = Set crackSpin(execute=@n33.then_1, crackSpin:float=float + float(A=Get crackSpin() -> (crackSpin:float), B=Get crackSpinSpeed() -> (crackSpinSpeed:float))) -> (Output_Get:float)
   > Set Spinning_Rotator_00(Spinning_Rotator_00:Rotator=MakeRotator(Pitch=NormalizeAxis!(Angle:float=@n34) -> (ReturnValue:float))) -> (Output_Get:Rotator)
   n35 = Branch(execute=@n33.then_0, Condition=Get ActiveKill())
   > n36 = Set crackSpinSpeed(crackSpinSpeed:float=FInterpTo(
     Current=Get crackSpinSpeed()
     Target=Get onCrackSpinSpeed() -> (onCrackSpinSpeed:float)
     DeltaTime=@n4.DeltaTimeX
     InterpSpeed=Get interpSpeed() -> (interpSpeed:float)
    )) -> (Output_Get:float)
   n37 = float - float(A=@n36, B=Get onCrackSpinSpeed()) {is=title K2Node_CallFunction}
   > Branch(Condition="float < float"!(A:double="Absolute (Float)"(A=@n37), B:double=0.1) -> (ReturnValue:bool) {is=title K2Node_CallFunction})
   > Set ActiveKill(ActiveKill=false)
   Set crackSpinSpeed(
    execute=@n35.else
    crackSpinSpeed=FInterpTo(Current=Get crackSpinSpeed(), Target=0.5, DeltaTime=@n4.DeltaTimeX, InterpSpeed=Get interpSpeed())
   )
   RetriggerableDelay(Duration=0.24)
  ]}
 spinFNC_deprecated {nodes [
   n1 = spinFNC_deprecated!() -> (then, speed:double=~, initVal:double=~) {is=title K2Node_FunctionEntry}
   n2 = float * float(A=@n1, B=GetRealTimeSeconds())
   > Return Node!(execute, output:double=NormalizeAxis(Angle=float + float(A=@n1.initVal, B=@n2)))
  ]}
]
variables [
 Pawn {type=FortPlayerPawn}
 Skel_Mesh {type=SkeletalMeshComponent}
]
